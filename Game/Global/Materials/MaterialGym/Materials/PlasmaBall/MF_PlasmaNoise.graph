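# Material function that outputs animated, colored plasma lightning with a center glow, for a plasma ball material.
# A panned turbulence flowmap, scaled by view-facing angle, warps UVs for three rotated, panned noise samples that are combined, screen-blended and sharpened into lightning.
# Lightning, tinted by falloff and fresnel colors, is added over layered base and center-glow gradients.
# The flowmap also perturbs the fresnel normal, skipped on low and medium quality in favor of the vertex normal.
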MaterialFunction /Game/Global/Materials/MaterialGym/Materials/PlasmaBall/MF_PlasmaNoise
graph {nodes [
  n1 = NamedRerouteDeclaration(None=FunctionInput(Preview=TextureCoordinate()) {bUsePreviewValueAsDefault=true, inputName=UV, inputType=FunctionInput_Vector2}) {name=UVInput, nodeColor=<rgba 0 0.4375 1 1>}
  n2 = ScalarParameter() {defaultValue=0.5, parameterName=Flowmap UV Scale, sliderMax=1}
  n3 = Constant()
  n4 = Multiply(A=ScalarParameter() {defaultValue=0.1, parameterName=Flowmap Panner Speed}, B=TwoSidedSign())
  n5 = TextureSample(UVs=Panner(Coordinate=Multiply(A=@n1, B=@n2), Speed=AppendVector(A=@n3, B=@n4)) {bFractionalPart=true}) {samplerType=SAMPLERTYPE_LinearColor, texture=/Game/Effects/Fort_Effects/Textures/Plasma/T_Turbulence_AdditiveWarp_Linear}
  n6 = Abs(None=DotProduct(A=PixelNormalWS(), B=CameraVectorWS()))
  n7 = Multiply(A=ConstantBiasScale(None=ComponentMask(None=@n5.RGB) {g=true, r=true}) {bias=-0.5, scale=2}, B=SquareRoot(None=@n6))
  n8 = Multiply(A=ScalarParameter() {defaultValue=0.066, parameterName=UV Flowmap Intensity}, B=@n7)
  n9 = StaticBool() {value=true}
  n10 = MaterialFunctionCall!(
   UV=Divide(A=Add(A=@n1, B=@n8) {constB=0}, B=ScalarParameter() {defaultValue=2.618, parameterName=Caustics UV Scale})
   Rotation=Constant() {r=0.25}
   Speed=ScalarParameter() {defaultValue=0.066, parameterName=Caustics Speed}
   Time
   Rotate UVs=@n9
   Fractional Part=@n9
  ) {materialFunction=/Game/Accessories/FORT_Backpacks/Backpack_M_MED_PitGlass/FX/Materials/MF_3UVPanner}
  n11 = TextureObject() {
   samplerType=SAMPLERTYPE_LinearGrayscale
   texture=/Game/Accessories/FORT_Backpacks/Backpack_M_MED_PitGlass/FX/Textures/T_Backpack_PitGlass_Noise
  }
  n12 = ScalarParameter() {defaultValue=1, parameterName=Texture Exponent}
  n13 = Add(
   A=Power(Base=TextureSample(UVs=@n10.UV1, Tex=@n11) {texture=/Game/Effects/Fort_Effects/Textures/Tile/T_ShieldCaustics_Tile_Chromatic}.R, Exp=@n12)
   B=Power(Base=TextureSample(UVs=@n10.UV2, Tex=@n11) {texture=&T_ShieldCaustics_Tile_Chromatic}.R, Exp=@n12)
  )
  n14 = Power(Base=TextureSample(UVs=@n10.UV3, Tex=@n11) {texture=&T_ShieldCaustics_Tile_Chromatic}.R, Exp=@n12)
  n15 = MaterialFunctionCall!(Blend=NamedRerouteUsage(), Base=Divide(A=Add(A=@n13, B=@n14)) {constB=3}) {materialFunction=/Engine/Functions/Engine_MaterialFunctions03/Blends/Blend_Screen}
  n16 = ScalarParameter() {defaultValue=3, parameterName=Blended Texture Exponent}
  n17 = Power(Base=@n15.Result, Exp=@n16)
  n18 = LinearInterpolate(
   A=VectorParameter() {parameterName=Color Lightning Fresnel, defaultValue=<rgba 1.3 0 3 0.1>}.RGB
   B=LinearInterpolate(
    A=VectorParameter() {parameterName=Color Lightning Falloff, defaultValue=<rgba 0.015 0 0.9 3>}.RGB
    B=VectorParameter() {parameterName=Color Lightning, defaultValue=<rgba 0.07 0.845 1 0>}.RGB
    Alpha=MaterialFunctionCall(In=@n15.Result, Power=@n16) {materialFunction=&SCurve}.Result
   )
   Alpha=Power(Base=NamedRerouteUsage(), Exp=ScalarParameter() {defaultValue=1, parameterName=Color Lightning Fresnel Power})
  )
  n19 = VectorParameter() {parameterName=Color C, defaultValue=<rgba 0.0008 0 0.01 2.5>}
  n20 = VectorParameter() {parameterName=Color A, defaultValue=<rgba 0.048 0.816 1.2 1>}
  n21 = NamedRerouteUsage()
  n22 = NamedRerouteUsage()
  n23 = LinearInterpolate(
   A=@n19.RGB
   B=LinearInterpolate(A=VectorParameter() {parameterName=Color B, defaultValue=<rgba 0.1332 0.6037 0.666 1>}.RGB, B=@n20.RGB, Alpha=Power(Base=@n21, Exp=@n20.A))
   Alpha=Power(Base=OneMinus(None=@n22), Exp=@n19.A)
  )
  n24 = VectorParameter() {parameterName=Center Glow Color A, defaultValue=<rgba 0.05 0.2075 0.5 8>}
  n25 = NamedRerouteUsage()
  n26 = LinearInterpolate(A=@n24.RGB, B=VectorParameter() {parameterName=Center Glow Color B, defaultValue=<rgba 0 6.6 33 8>}.RGB, Alpha=Power(Base=@n25, Exp=@n24.A))
  FunctionOutput(None=Add(A=Multiply(A=@n17, B=@n18), B=LinearInterpolate(A=@n23, B=@n26, Alpha=@n25) {constB=0}) {constB=0})
  n27 = OneMinus(None=NamedRerouteUsage())
  n28 = Saturate(None=ScalarParameter() {defaultValue=0.75, parameterName=Center Glow Size})
  n29 = OneMinus(None=Saturate(None=Divide(A=@n27, B=@n28)))
  n30 = ScalarParameter() {defaultValue=22, parameterName=Center Glow Exponent}
  n31 = Constant() {r=1}
  NamedRerouteDeclaration(None=Multiply(
    A=Power(Base=@n29, Exp=@n30)
    B=FunctionInput(Preview=@n31) {bUsePreviewValueAsDefault=true, inputName=Center Glow brightness, inputType=FunctionInput_Scalar, sortPriority=2}
   )) {name=Center Glow Mask, nodeColor=<rgba 1 0.8203 0 1>}
  n32 = LinearInterpolate(
   A=ScalarParameter() {defaultValue=0.5, parameterName=Fresnel Flowmap Intensity Max}
   B=ScalarParameter() {defaultValue=0.1, parameterName=Fresnel Flowmap Intensity Min}
   Alpha=@n6
  )
  n33 = Transform(None=AppendVector(A=Multiply(A=@n7, B=@n32), B=Constant())) {transformSourceType=TRANSFORMSOURCE_View}
  n34 = PixelNormalWS()
  n35 = VertexNormalWS()
  n36 = QualitySwitch(Default=Normalize(VectorInput=Add(A=@n33, B=@n34)), Low=@n35, Medium=@n35) {desc=Skipping flowmap on med/low}
  n37 = CameraVectorWS()
  NamedRerouteDeclaration(None=Abs(None=DotProduct(A=@n36, B=@n37))) {name=Fresnel, nodeColor=<rgba 0.7578 1 0 1>}
 ]}
