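# M_PlasmaNoiseExample is a Default Lit example material for the MF_PlasmaNoise material function, kept in the PlasmaBall folder of the Material Gym.
# Its graph holds one call to that function, which exposes UV and center glow brightness inputs, plus a lone Constant node, with no links shown between nodes or to material outputs.
# Automatic usage setting in the editor is off, with skeletal mesh and mesh deformer usage enabled.
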Material /Game/Global/Materials/MaterialGym/Materials/PlasmaBall/M_PlasmaNoiseExample
graph {
 properties {bAutomaticallySetUsageInEditor=false, bUsedWithMeshDeformer=true, bUsedWithSkeletalMesh=true, usedShadingModels=Default Lit}
 nodes [
  Constant()
  MaterialFunctionCall!(UV, Center Glow brightness) {materialFunction=&/MF_PlasmaNoise}
 ]
}
